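# Niagara glow effect for the GreenHazard backpack cosmetic, drawing green sprites on the backpack's sockets.
# CapsuleGlows bursts ten persistent particles, partitioned one per socket with per-socket size and camera offset, that follow socket position and align to its rotation.
# Two sprite renderers accompany separately computed base and light colors and sizes.
# Colors scale by a system-level fade, and alpha also dims as the camera gets close.
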NiagaraSystem /BRCosmetics/Accessories/FORT_Backpacks/Backpack_M_MED_GreenHazard/FX/NS_Backpack_GreenHazard
dependencies {
 usedDataInterfaces [&NiagaraDataInterfaceCamera, &NiagaraDataInterfaceSocketReader, &NiagaraDataInterfaceCurve]
 usedDynamicInputs [&FNTimeOfDayScalar, &FloatFromCurve]
 usedModules [
  &self:SystemSpawnScript.SystemScriptSource.SystemScriptGraph.NiagaraNodeAssignment_0.SetVariables_#1
  &SystemState
  &self:SystemSpawnScript.SystemScriptSource.SystemScriptGraph.NiagaraNodeAssignment_3.SetVariables_#2
  &self:SystemSpawnScript.SystemScriptSource.SystemScriptGraph.NiagaraNodeAssignment_2.SetVariables_#3
 ]
}
emitters {CapsuleGlows {
  emitterUpdateScript [
   &EmitterState(Life Cycle Mode=::NewEnumerator1, Loop Behavior=::NewEnumerator1, Loop Duration Mode=::NewEnumerator1, UseLoopDelay=false)
   &SpawnBurst_Instantaneous(Spawn Count=10, Use Spawn Probability=false, UseLoopCountLimit=true)
  ]
  particleSpawnScript [
   &PartitionParticles(ID to Partition:NiagaraInt32=&ReturnExecIndex(), Number Of Partitions=10)
   &InitializeParticle(
    Color=<rgba 0.4575 1 0.07 0.5>
    Position Mode=::NewEnumerator0
    Sprite Rotation Mode=::NewEnumerator0
    Sprite Size=&MergeHelper_Vector2Switch(
     0=<9 10.8>
     1=<9 10.8>
     2=<9 10.8>
     3=<9 10.8>
     4=<4 4>
     5=<3 3>
     6=<2 2>
     7=<2 2>
     8=<5 7>
     9=<5 7>
     Index=$Particles.PartitionParticles.Partition
     Num=::NewEnumerator8
    )
    Sprite Size Mode=::NewEnumerator4
    Write Material Random=true
   )
   &CameraOffset(Camera Offset Amount=&MergeHelper_FloatSwitch(0=5, 1=5, 2=5, 3=5, 4=1, 5=1, 6=1, 7=1, 8=0, 9=0, Index=$Particles.PartitionParticles.Partition, Num=::NewEnumerator8))
   /SetVariables(Particles.SpriteSizeLight:Vector2f=&MergeHelper_Vector2Switch(
     0=<15 15>
     1=<15 15>
     2=<10 10>
     3=<10 10>
     4=<6 5>
     5=<3 3>
     6=<3 3>
     7=<3 3>
     8=<0 0>
     9=<0 0>
     Index=$Particles.PartitionParticles.Partition
     Num=::NewEnumerator8
    )) {isSetParameters=true}
  ]
  particleUpdateScript [
   /SetVariables(Particles.CameraDistanceAlpha:NiagaraFloat=&RemapRange(
     Input Value To Remap=/Niagara/DynamicInputs/Camera/DistanceToCamera(Use Owner Position:NiagaraBool=true, Camera Query:NiagaraDataInterfaceCamera=DataInterface())
     Input Min=25
     Input Max=50
     Output Min=0.1
     Output Max=1
    )) {isSetParameters=true}
   &SocketLocation(
    SamplePosOffset=true
    Socket Index=$Particles.PartitionParticles.Partition
    Socket Reader=$System.Socket Reader
    Socket Sampling Mode=::NewEnumerator3
    VelocitySpeedLimit=false
   )
   &SpriteFacingAndAlignment(
    Sprite Alignment=&RotateVectorByQuaternion(Vector To Rotate=<0 1 0>, Quaternion=$Output.SocketLocation.SampledRotation)
    WriteAlignment=true
    WriteFacing=false
   )
   &ParticleState(Kill Particles When Lifetime Has Elapsed=false, Let Infinitely Lived Particles Die When Emitter Deactivates=true)
   &DynamicMaterialParameters(
    Index 0 Param 1=2.5
    Index 0 Param 2=10
    Module.Index0Param0Write=true
    Module.Index0Param1Write=false
    Module.Index0Param2Write=false
    Module.Index0Param3Write=false
   ) {module=DynamicMaterialParameters001}
   /SetVariables(Particles.ColorLight:LinearColor=&MultiplyLinearColor(
     Linear Color=<rgba 2.46 4 1.36 1>
     Scale Factor=&MakeVector4FromVectorAndFloat(
      "Vector (RGB)"=&VectorFromFloat(Value=$System.SystemFade)
      "Float (Alpha)"=&Multiply_Float(A=$System.SystemFade, B=$Particles.CameraDistanceAlpha)
     )
    )) {isSetParameters=true}
   &ScaleColor(
    Scale Alpha=&Multiply_Float(A=$System.SystemFade, B=$Particles.CameraDistanceAlpha)
    Scale RGB=&VectorFromFloat(Value=$System.SystemFade)
    ScaleA=true
    ScaleRGB=true
   )
  ]
  rendererClasses [&NiagaraSpriteRendererProperties]
  renderers [
   {rendererClass=&NiagaraSpriteRendererProperties}
   {rendererClass=&NiagaraSpriteRendererProperties, rendererIndex=1}
  ]
 }}
summary {
 systemName=NS_Backpack_GreenHazard
 emitters [{emitterName=CapsuleGlows, rendererClasses [&NiagaraSpriteRendererProperties]}]
 userVariables [
  User.bIsFrontEnd {type=NiagaraBool}
  User.bIsFrontEndPreview {type=NiagaraBool}
 ]
}
