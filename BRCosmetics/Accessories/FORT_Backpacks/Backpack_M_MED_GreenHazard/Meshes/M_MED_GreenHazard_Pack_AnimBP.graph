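# Animation Blueprint for the GreenHazard backpack cosmetic that fits the pack to the wearer's torso.
# Each update copies pelvis and spine world transforms from the attach parent and maps spine bend to a BackpackRotate alpha.
# The AnimGraph offsets the pack bone (also by front-end state and that alpha) and drives proxy "constarane" bones with those transforms.
# Weighted constraints position Dyn_spine bones from the proxies, then Dyn_spine_01 is re-constrained to pack.
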AnimBlueprint /BRCosmetics/Accessories/FORT_Backpacks/Backpack_M_MED_GreenHazard/Meshes/M_MED_GreenHazard_Pack_AnimBP
target_skeleton=&/M_MED_GreenHazard_Pack_Skeleton
graphs [
 AnimGraph {nodes [
   n1 = "Transform (Modify) Bone - Bone: pack"(ComponentPose=Local To Component(LocalPose=Input Pose()), Translation="0,-1,-5", Rotation="0,0,13") {
    rotationMode=BMM_Additive
    scaleMode=BMM_Additive
    translationMode=BMM_Additive
    boneToModify {boneName=pack}
    rotation {roll=13}
    translation {y=-1, z=-5}
   }
   n2 = "Transform (Modify) Bone - Bone: pack"!(ComponentPose=@n1, bAlphaBoolEnabled:bool=Get bIsInFrontEnd(), Translation="0,2,0", Rotation="0,0,0", Scale) -> (Pose) {
    alphaInputType=Bool
    rotationMode=BMM_Additive
    scaleMode=BMM_Additive
    translationMode=BMM_Additive
    boneToModify {boneName=pack}
    translation {y=2}
   }
   n3 = "Transform (Modify) Bone - Bone: pack"!(ComponentPose=@n2, Alpha:float=Get BackpackRotate() -> (BackpackRotate:double), Translation="0,-5,3", Rotation="0,0,20", Scale) -> (Pose) {
    rotationMode=BMM_Additive
    scaleMode=BMM_Additive
    translationMode=BMM_Additive
    boneToModify {boneName=pack}
    rotation {roll=20}
    translation {y=-5, z=3}
   }
   n4 = "Transform (Modify) Bone - Bone: constarane_pelvis"(
    ComponentPose=@n3
    Alpha=1
    Translation=Get pelvis_01_tran() -> (pelvis_01_tran:Vector)
    Rotation=Get pelvis_01_rot() -> (pelvis_01_rot:Rotator)
    Scale="1,1,1"
   ) {
    rotationMode=BMM_Replace
    rotationSpace=BCS_WorldSpace
    scaleSpace=BCS_BoneSpace
    translationMode=BMM_Replace
    translationSpace=BCS_WorldSpace
    alphaBoolBlend {blendOption=Linear}
    alphaScaleBiasClamp {inRange {max=0, min=-40}, outRange {max=0, min=1}}
    boneToModify {boneName=constarane_pelvis}
   }
   n5 = "Transform (Modify) Bone - Bone: constarane_spine_01"(
    ComponentPose=@n4
    Alpha=1
    Translation=Get spine_01_tran() -> (spine_01_tran:Vector)
    Rotation=Get spine_01_rot() -> (spine_01_rot:Rotator)
    Scale="1,1,1"
   ) {
    rotationMode=BMM_Replace
    rotationSpace=BCS_WorldSpace
    scaleSpace=BCS_BoneSpace
    translationMode=BMM_Replace
    translationSpace=BCS_WorldSpace
    alphaBoolBlend {blendOption=Linear}
    alphaScaleBiasClamp {inRange {max=0, min=-40}, outRange {max=0, min=1}}
    boneToModify {boneName=constarane_spine_01}
   }
   n6 = "Transform (Modify) Bone - Bone: constarane_spine_02"(
    ComponentPose=@n5
    Alpha=1
    Translation=Get spine_02_tran() -> (spine_02_tran:Vector)
    Rotation=Get spine_02_rot() -> (spine_02_rot:Rotator)
    Scale="1,1,1"
   ) {
    rotationMode=BMM_Replace
    rotationSpace=BCS_WorldSpace
    scaleSpace=BCS_BoneSpace
    translationMode=BMM_Replace
    translationSpace=BCS_WorldSpace
    alphaBoolBlend {blendOption=Linear}
    alphaScaleBiasClamp {inRange {max=0, min=-40}, outRange {max=0, min=1}}
    boneToModify {boneName=constarane_spine_02}
   }
   n7 = "Transform (Modify) Bone - Bone: constarane_spine_03"(
    ComponentPose=@n6
    Alpha=1
    Translation=Get spine_03_tran() -> (spine_03_tran:Vector)
    Rotation=Get spine_03_rot() -> (spine_03_rot:Rotator)
    Scale="1,1,1"
   ) {
    rotationMode=BMM_Replace
    rotationSpace=BCS_WorldSpace
    scaleSpace=BCS_BoneSpace
    translationMode=BMM_Replace
    translationSpace=BCS_WorldSpace
    alphaBoolBlend {blendOption=Linear}
    alphaScaleBiasClamp {inRange {max=0, min=-40}, outRange {max=0, min=1}}
    boneToModify {boneName=constarane_spine_03}
   }
   n8 = "Transform (Modify) Bone - Bone: constarane_spine_04"(
    ComponentPose=@n7
    Alpha=1
    Translation=Get spine_04_tran() -> (spine_04_tran:Vector)
    Rotation=Get spine_04_rot() -> (spine_04_rot:Rotator)
    Scale="1,1,1"
   ) {
    rotationMode=BMM_Replace
    rotationSpace=BCS_WorldSpace
    scaleSpace=BCS_BoneSpace
    translationMode=BMM_Replace
    translationSpace=BCS_WorldSpace
    alphaBoolBlend {blendOption=Linear}
    alphaScaleBiasClamp {inRange {max=0, min=-40}, outRange {max=0, min=1}}
    boneToModify {boneName=constarane_spine_04}
   }
   n9 = "Transform (Modify) Bone - Bone: constarane_spine_05"(
    ComponentPose=@n8
    Alpha=1
    Translation=Get spine_05_tran() -> (spine_05_tran:Vector)
    Rotation=Get spine_05_rot() -> (spine_05_rot:Rotator)
    Scale="1,1,1"
   ) {
    rotationMode=BMM_Replace
    rotationSpace=BCS_WorldSpace
    scaleSpace=BCS_BoneSpace
    translationMode=BMM_Replace
    translationSpace=BCS_WorldSpace
    alphaBoolBlend {blendOption=Linear}
    alphaScaleBiasClamp {inRange {max=0, min=-40}, outRange {max=0, min=1}}
    boneToModify {boneName=constarane_spine_05}
   }
   n10 = "Constraint - Dyn_spine_01 (PP)"!(ComponentPose=@n9, Alpha=1, ConstraintWeights_0=0.5, ConstraintWeights_1:float=0.5) -> (Pose) {
    alphaBoolBlend {blendOption=Linear}
    boneToModify {boneName=Dyn_spine_01}
    constraintSetup [{targetBone {boneName=constarane_spine_05}}, {targetBone {boneName=constarane_spine_04}}]
    constraintWeights [1,1]
   }
   n11 = "Constraint - Dyn_spine_02 (PP)"!(ComponentPose=@n10, Alpha=1, ConstraintWeights_0=1, ConstraintWeights_1:float=0) -> (Pose) {
    alphaBoolBlend {blendOption=Linear}
    boneToModify {boneName=Dyn_spine_02}
    constraintSetup [{targetBone {boneName=constarane_spine_04}}, {targetBone {boneName=constarane_spine_05}}]
    constraintWeights [1,1]
   }
   n12 = "Constraint - Dyn_spine_03 (PP)"!(ComponentPose=@n11, Alpha=1, ConstraintWeights_0=0.7, ConstraintWeights_1:float=0.3) -> (Pose) {
    alphaBoolBlend {blendOption=Linear}
    boneToModify {boneName=Dyn_spine_03}
    constraintSetup [{targetBone {boneName=constarane_spine_03}}, {targetBone {boneName=constarane_spine_04}}]
    constraintWeights [1,1]
   }
   n13 = "Constraint - Dyn_spine_04 (PP)"!(ComponentPose=@n12, Alpha=1, ConstraintWeights_0=0.5, ConstraintWeights_1:float=0.5) -> (Pose) {
    alphaBoolBlend {blendOption=Linear}
    boneToModify {boneName=Dyn_spine_04}
    constraintSetup [{targetBone {boneName=constarane_spine_02}}, {targetBone {boneName=constarane_spine_03}}]
    constraintWeights [1,1]
   }
   n14 = "Constraint - Dyn_spine_01 (P)"(ComponentPose=@n13, Alpha=1, ConstraintWeights_0=1) {
    alphaBoolBlend {blendOption=Linear}
    boneToModify {boneName=Dyn_spine_01}
    constraintSetup [{targetBone {boneName=pack}}]
    constraintWeights [1]
   }
   n15 = "Transform (Modify) Bone - Bone: Dyn_spine_03"!(ComponentPose=@n14, bAlphaBoolEnabled:bool=True, Translation="0,0,0", Rotation="0,0,4", Scale="1,1,1") -> (Pose) {
    alphaInputType=Bool
    rotationMode=BMM_Additive
    scaleMode=BMM_Additive
    translationMode=BMM_Additive
    boneToModify {boneName=Dyn_spine_03}
    rotation {roll=4}
   }
   Output Pose(Result=Component To Local(ComponentPose=@n15)) {layerGroup=None}
  ]}
 EventGraph {nodes [
   Event BlueprintUpdateAnimation()
   n1 = GetPartSkeletalMeshComponent()
   > Is Valid(InputObject=@n1)
   n2 = GetAttachParent!(self:Scene Component Object Reference=@n1) -> (ReturnValue:Scene Component Object Reference)
   > Is Valid(InputObject=@n2)
   > n3 = Sequence!(execute) -> (then_0, then_1, then_2:Exec, then_3:Exec, then_4:Exec, then_5:Exec, then_6:Exec)
   n4 = BreakTransform(InTransform=GetSocketTransform(self=@n2, InSocketName=spine_01, TransformSpace=RTS_World))
   > Set spine_01_tran(spine_01_tran:Vector=@n4) -> (Output_Get:Vector)
   > Set spine_01_rot(spine_01_rot:Rotator=@n4.Rotation) -> (Output_Get:Rotator)
   n5 = BreakTransform(InTransform=GetSocketTransform(self=@n2, InSocketName=spine_02, TransformSpace=RTS_World))
   Set spine_02_tran(execute=@n3.then_1, spine_02_tran:Vector=@n5) -> (Output_Get:Vector)
   > Set spine_02_rot(spine_02_rot:Rotator=@n5.Rotation) -> (Output_Get:Rotator)
   n6 = BreakTransform(InTransform=GetSocketTransform(self=@n2, InSocketName=spine_03, TransformSpace=RTS_World))
   Set spine_03_tran(execute=@n3.then_2, spine_03_tran:Vector=@n6) -> (Output_Get:Vector)
   > Set spine_03_rot(spine_03_rot:Rotator=@n6.Rotation) -> (Output_Get:Rotator)
   n7 = BreakTransform(InTransform=GetSocketTransform(self=@n2, InSocketName=spine_04, TransformSpace=RTS_World))
   Set spine_04_tran(execute=@n3.then_3, spine_04_tran:Vector=@n7) -> (Output_Get:Vector)
   > Set spine_04_rot(spine_04_rot:Rotator=@n7.Rotation) -> (Output_Get:Rotator)
   n8 = BreakTransform(InTransform=GetSocketTransform(self=@n2, InSocketName=spine_05, TransformSpace=RTS_World))
   Set spine_05_tran(execute=@n3.then_4, spine_05_tran:Vector=@n8) -> (Output_Get:Vector)
   > Set spine_05_rot(spine_05_rot:Rotator=@n8.Rotation) -> (Output_Get:Rotator)
   n9 = BreakTransform(InTransform=GetSocketTransform(self=@n2, InSocketName=pelvis, TransformSpace=RTS_World))
   Set pelvis_01_tran(execute=@n3.then_5, pelvis_01_tran:Vector=@n9) -> (Output_Get:Vector)
   > Set pelvis_01_rot(pelvis_01_rot:Rotator=@n9.Rotation) -> (Output_Get:Rotator)
   n10 = BreakTransform(InTransform=GetDeltaTransformFromRefPose(self=GetOwnerSkeletalMeshComponent(), BoneName=spine_05, BaseName=spine_01))
   Set BackpackRotate(execute=@n3.then_6, BackpackRotate:double=MapRangeClamped(Value=BreakRotator(InRot=@n10.Rotation).Yaw, InRangeA=0, InRangeB=-50)) -> (Output_Get:double)
   TryGetPawnOwner()
  ]}
]
variables [
 BackpackRotate {type=number}
 pelvis_01_rot {type=Rotator}
 pelvis_01_tran {type=Vector}
 spine_01_rot {type=Rotator}
 spine_01_tran {type=Vector}
 spine_02_rot {type=Rotator}
 spine_02_tran {type=Vector}
 spine_03_rot {type=Rotator}
 spine_03_tran {type=Vector}
 spine_04_rot {type=Rotator}
 spine_04_tran {type=Vector}
 spine_05_rot {type=Rotator}
 spine_05_tran {type=Vector}
]
